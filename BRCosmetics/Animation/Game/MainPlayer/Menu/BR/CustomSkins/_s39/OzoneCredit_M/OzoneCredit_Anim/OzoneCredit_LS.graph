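# Level Sequence for the OzoneCredit_M custom skin's menu animation, apparently a scene for animating or previewing it.
# A FortPlayerMannequin binding with a spawn track carries a skeletal animation section and a Fortnite_Legacy_Face_CtrlRig track over the same frame range at 30 fps.
# The face rig section shows only channel defaults, which place and scale the master face control and set head following, with Flipbook off.
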LevelSequence /BRCosmetics/Animation/Game/MainPlayer/Menu/BR/CustomSkins/_s39/OzoneCredit_M/OzoneCredit_Anim/OzoneCredit_LS
bindings [FortPlayerMannequin {
  id=#1
  possessed_class=FortPlayerMannequin
  tracks [
   {type=MovieSceneSpawnTrack, sections [{type=MovieSceneSpawnSection, channels [None {type=MovieSceneScriptingBoolChannel, default=true}]}]}
   {type=MovieSceneSkeletalAnimationTrack, sections [{type=MovieSceneSkeletalAnimationSection, end_frame=331, start_frame=1}]}
   {
    type=MovieSceneControlRigParameterTrack
    display_name=Fortnite_Legacy_Face_CtrlRig
    sections [{
      type=MovieSceneControlRigParameterSection
      end_frame=331
      start_frame=1
      channels [
       Flipbook {type=MovieSceneScriptingBoolChannel, default=false}
       Face_Ctrl_Master.Location.X {type=MovieSceneScriptingFloatChannel, default=73.9131}
       Face_Ctrl_Master.Location.Y {type=MovieSceneScriptingFloatChannel, default=-0.3613}
       Face_Ctrl_Master.Location.Z {type=MovieSceneScriptingFloatChannel, default=-43.713}
       Face_Ctrl_Master.Rotation.X {type=MovieSceneScriptingFloatChannel, default=0}
       Face_Ctrl_Master.Rotation.Y {type=MovieSceneScriptingFloatChannel, default=0}
       Face_Ctrl_Master.Rotation.Z {type=MovieSceneScriptingFloatChannel, default=0}
       Face_Ctrl_Master.Scale.X {type=MovieSceneScriptingFloatChannel, default=5}
       Face_Ctrl_Master.Scale.Y {type=MovieSceneScriptingFloatChannel, default=5}
       Face_Ctrl_Master.Scale.Z {type=MovieSceneScriptingFloatChannel, default=5}
       gui_follow_head {type=MovieSceneScriptingFloatChannel, default=0.3056}
      ]
     }]
   }
  ]
 }]
display_rate {denominator=1, numerator=30}
tick_resolution {denominator=1, numerator=24000}
tracks []
